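# Physics asset for the female medium Skirmish Alt body that provides secondary motion for the shirt's dynamic bones.
# Eight shirt capsules ring the front, back, sides and diagonals, each constrained to its own tiny _MSTR body with moderate swing and tight twist limits.
# Each _MSTR body links to the spine_03 torso capsule with free swing and twist, so all eight chains share one anchor.
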PhysicsAsset /BRCosmetics/Characters/Player/Female/Medium/Bodies/F_MED_Skirmish_Alt/Meshes/F_MED_Skirmish_Alt_Dyn
bodies {
 dyn_shirt_bk [{shapeType=Capsule, length=4.875, radius=5.125, center=<3 0 4.5>, rotation=<rot 90 0 0>}]
 dyn_shirt_bk_MSTR [{shapeType=Capsule, length=0.505, radius=0.505}]
 dyn_shirt_bk_l [{shapeType=Capsule, length=5.2188, radius=4.7812, center=<3 0 4.5>, rotation=<rot 90 0 0>}]
 dyn_shirt_bk_l_MSTR [{shapeType=Capsule, length=0.505, radius=0.505}]
 dyn_shirt_bk_r [{shapeType=Capsule, length=5.2188, radius=4.7812, center=<3 0 4.5>, rotation=<rot 90 0 0>}]
 dyn_shirt_bk_r_MSTR [{shapeType=Capsule, length=0.505, radius=0.505}]
 dyn_shirt_fr [{shapeType=Capsule, length=5, radius=5, center=<3 0 4.5>, rotation=<rot 90 0 0>}]
 dyn_shirt_fr_MSTR [{shapeType=Capsule, length=0.505, radius=0.505}]
 dyn_shirt_fr_l [{shapeType=Capsule, length=5.875, radius=4.125, center=<3 0 4.5>, rotation=<rot 90 0 0>}]
 dyn_shirt_fr_l_MSTR [{shapeType=Capsule, length=0.505, radius=0.505}]
 dyn_shirt_fr_r [{shapeType=Capsule, length=5.875, radius=4.125, center=<3 0 4.5>, rotation=<rot 90 0 0>}]
 dyn_shirt_fr_r_MSTR [{shapeType=Capsule, length=0.505, radius=0.505}]
 dyn_shirt_l [{shapeType=Capsule, length=5, radius=5, center=<3 0 4.5>, rotation=<rot 90 0 0>}]
 dyn_shirt_l_MSTR [{shapeType=Capsule, length=0.505, radius=0.505}]
 dyn_shirt_r [{shapeType=Capsule, length=5, radius=5, center=<3 0 4.5>, rotation=<rot 90 0 0>}]
 dyn_shirt_r_MSTR [{shapeType=Capsule, length=0.505, radius=0.505}]
 spine_03 [{shapeType=Capsule, length=6.25, radius=9.4738, center=<3.2754 2.3858 0>}]
}
constraints [
 {bone1Name=dyn_shirt_fr, bone2Name=dyn_shirt_fr_MSTR, swing1LimitDegrees=45, swing2LimitDegrees=45, twistLimitDegrees=5}
 {bone1Name=dyn_shirt_fr_l, bone2Name=dyn_shirt_fr_l_MSTR, swing1LimitDegrees=45, swing2LimitDegrees=45, twistLimitDegrees=5}
 {bone1Name=dyn_shirt_l, bone2Name=dyn_shirt_l_MSTR, swing1LimitDegrees=45, swing2LimitDegrees=45, twistLimitDegrees=5}
 {bone1Name=dyn_shirt_bk_l, bone2Name=dyn_shirt_bk_l_MSTR, swing1LimitDegrees=45, swing2LimitDegrees=45, twistLimitDegrees=5}
 {bone1Name=dyn_shirt_bk, bone2Name=dyn_shirt_bk_MSTR, swing1LimitDegrees=45, swing2LimitDegrees=45, twistLimitDegrees=5}
 {bone1Name=dyn_shirt_fr_r, bone2Name=dyn_shirt_fr_r_MSTR, swing1LimitDegrees=45, swing2LimitDegrees=45, twistLimitDegrees=5}
 {bone1Name=dyn_shirt_r, bone2Name=dyn_shirt_r_MSTR, swing1LimitDegrees=45, swing2LimitDegrees=45, twistLimitDegrees=5}
 {bone1Name=dyn_shirt_bk_r, bone2Name=dyn_shirt_bk_r_MSTR, swing1LimitDegrees=45, swing2LimitDegrees=45, twistLimitDegrees=5}
 {
  bone1Name=dyn_shirt_fr_MSTR
  bone2Name=spine_03
  swing1LimitDegrees=45
  swing1Motion=Free
  swing2LimitDegrees=45
  swing2Motion=Free
  twistLimitDegrees=45
  twistMotion=Free
 }
 {
  bone1Name=dyn_shirt_fr_l_MSTR
  bone2Name=spine_03
  swing1LimitDegrees=45
  swing1Motion=Free
  swing2LimitDegrees=45
  swing2Motion=Free
  twistLimitDegrees=45
  twistMotion=Free
 }
 {
  bone1Name=dyn_shirt_l_MSTR
  bone2Name=spine_03
  swing1LimitDegrees=45
  swing1Motion=Free
  swing2LimitDegrees=45
  swing2Motion=Free
  twistLimitDegrees=45
  twistMotion=Free
 }
 {
  bone1Name=dyn_shirt_bk_l_MSTR
  bone2Name=spine_03
  swing1LimitDegrees=45
  swing1Motion=Free
  swing2LimitDegrees=45
  swing2Motion=Free
  twistLimitDegrees=45
  twistMotion=Free
 }
 {
  bone1Name=dyn_shirt_bk_MSTR
  bone2Name=spine_03
  swing1LimitDegrees=45
  swing1Motion=Free
  swing2LimitDegrees=45
  swing2Motion=Free
  twistLimitDegrees=45
  twistMotion=Free
 }
 {
  bone1Name=dyn_shirt_fr_r_MSTR
  bone2Name=spine_03
  swing1LimitDegrees=45
  swing1Motion=Free
  swing2LimitDegrees=45
  swing2Motion=Free
  twistLimitDegrees=45
  twistMotion=Free
 }
 {
  bone1Name=dyn_shirt_r_MSTR
  bone2Name=spine_03
  swing1LimitDegrees=45
  swing1Motion=Free
  swing2LimitDegrees=45
  swing2Motion=Free
  twistLimitDegrees=45
  twistMotion=Free
 }
 {
  bone1Name=dyn_shirt_bk_r_MSTR
  bone2Name=spine_03
  swing1LimitDegrees=45
  swing1Motion=Free
  swing2LimitDegrees=45
  swing2Motion=Free
  twistLimitDegrees=45
  twistMotion=Free
 }
]
