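# Post-pose AnimBlueprint for the female medium Skirmish Alt body part, adding deformation and secondary physics on top of the attached parent mesh's copied pose.
# Two Control Rigs from the base Skirmish body apply twist and sculpter corrections, then a RigidBody node simulates the Dyn physics asset relative to dyn_root_simspace.
# Each update, deprecated helper functions compute a state-based gravity and wind override plus an interpolated sim-space rotation that drives that bone in world space.
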AnimBlueprint /BRCosmetics/Characters/Player/Female/Medium/Bodies/F_MED_Skirmish_Alt/Meshes/F_MED_Skirmish_Alt_AnimBP
target_skeleton=&/F_MED_Skirmish_Alt_Skeleton
graphs [
 AnimGraph {nodes [
   n1 = Control Rig(Source=Copy Pose From Mesh() {bUseAttachedParent=true}) {
    alphaBoolBlend {blendOption=Linear}
    controlRigAssetReference {blueprintRigClass=/Game/Characters/Player/Female/Medium/Bodies/F_MED_Skirmish/Meshes/F_MED_Skirmish_Deform_CtrlRig_twists.F_MED_Skirmish_Deform_CtrlRig_twists_C}
   }
   n2 = Control Rig!(
    Source=@n1
    Alpha
    KneeRot_fr_l_114deg:Vector
    KneeTrans_fr_l_114deg:Vector="0.258,-3.591,0.009"
    KneeScale_fr_l_114deg:Vector="1,1,1"
    KneeScale_fr_l_130deg:Vector="1,1,1"
    KneeRot_fr_l_130deg:Vector="0,0,-51.2"
    KneeTrans_fr_l_130deg:Vector="0.258,-3.591,0.009"
   ) -> (Pose) {
    alphaBoolBlend {blendOption=Linear}
    controlRigAssetReference {blueprintRigClass=/Game/Characters/Player/Female/Medium/Bodies/F_MED_Skirmish/Meshes/F_MED_Skirmish_Deform_CtrlRig_sculpters.F_MED_Skirmish_Deform_CtrlRig_sculpters_C}
   }
   n3 = "Transform (Modify) Bone - Bone: dyn_root_simspace"(
    ComponentPose=Local To Component(LocalPose=@n2)
    Alpha=1
    Translation="0,0,0"
    Rotation=Get sim_space_rotation() -> (sim_space_rotation:Rotator)
    Scale="1,1,1"
   ) {
    rotationMode=BMM_Replace
    rotationSpace=BCS_WorldSpace
    translationSpace=BCS_BoneSpace
    alphaBoolBlend {blendOption=Linear}
    alphaScaleBiasClamp {bInterpResult=true, interpSpeedDecreasing=2, interpSpeedIncreasing=2}
    boneToModify {boneName=dyn_root_simspace}
   }
   n4 = RigidBody!(ComponentPose=@n3, Alpha, OverrideWorldGravity:Vector=Get GravityOverride(), ExternalForce) -> (Pose) {
    bClampLinearTranslationLimitToRefPose=true
    bOverrideWorldGravity=true
    overridePhysicsAsset=&/F_MED_Skirmish_Alt_Dyn
    simulationSpace=BaseBoneSpace
    alphaBoolBlend {blendOption=Linear}
    baseBoneRef {boneName=dyn_root_simspace}
    componentAppliedLinearAccClamp=<3000 3000 3000>
    componentLinearAccScale {z=0.5}
    componentLinearVelScale {z=0.5}
    overrideWorldGravity {z=-980}
    simSpaceSettings {
     maxAngularAcceleration=1000
     maxAngularVelocity=3
     maxLinearAcceleration=700
     maxLinearVelocity=500
     velocityScaleZ=0
     worldAlpha=1
     externalLinearDragV=<1 1>
    }
   }
   Output Pose(Result=Component To Local(ComponentPose=@n4)) {layerGroup=None}
  ]}
 EventGraph {nodes [
   n1 = Event BlueprintUpdateAnimation()
   > n2 = Sequence()
   > n3 = Gravity Override State Function DEPRECATED(
    CharacterPartAnimInstance=Self-Reference()
    GameTimeInSeconds=GetGameTimeInSeconds()
    GravityOverrideParams=Get Gravity Override Params()
    DebugJointName=spine_05
   )
   > Set GravityOverride(GravityOverride=@n3)
   n4 = ComplexSimSpaceFunction DEPRECATED!(
    execute=@n2.then_1
    CharacterPartAnimInstance:Custom Character Part Anim Instance Object Reference=Self-Reference()
    InterpSpeedMax:double=9
    PawnSpeedMax:double=20
    RInterpErrorTolerance:double=1
    DeltaTime:double=@n1.DeltaTimeX
    DebugDraw:bool=false
    sim_space_rotation_in:Rotator=Get sim_space_rotation()
    inComponentSpaceIn:bool=Get InComponentSpace() -> (InComponentSpace:bool)
    inComponentSpaceLastFrameIn:bool=Get InComponentSpaceLastFrame() -> (InComponentSpaceLastFrame:bool)
    lockedOnSimSpaceIn:bool=Get LockedOnSimSpace() -> (LockedOnSimSpace:bool)
    InterpSpeedIn:double=Get InterpSpeed() -> (InterpSpeed:double)
   ) -> (
    then
    sim_space_rotation:Rotator
    inComponentSpaceOut:bool
    inComponentSpaceLastFrameOut:bool
    lockedOnSimSpaceOut:bool
    InterpSpeedOut:double
   )
   > Set sim_space_rotation(sim_space_rotation:Rotator=@n4) -> (Output_Get:Rotator)
   > Set InComponentSpace(InComponentSpace:bool=@n4.inComponentSpaceOut) -> (Output_Get:bool)
   > Set InComponentSpaceLastFrame(InComponentSpaceLastFrame:bool=@n4.inComponentSpaceLastFrameOut) -> (Output_Get:bool)
   > Set LockedOnSimSpace(LockedOnSimSpace:bool=@n4.lockedOnSimSpaceOut) -> (Output_Get:bool)
   > Set InterpSpeed(InterpSpeed:double=@n4.InterpSpeedOut) -> (Output_Get:double)
  ]}
]
variables [
 Gravity Override Params {
  type=GravityOverrideParamsStruct
  default {
   emoteOrMelee {gravityOverride=<0 -50 -15>, windAmplitude=<50 150 300>, windFrequency=<1.5 2 6>, windOffset=<0 -200 -200>}
   front_End {gravityOverride=<0 -50 -15>, windAmplitude=<50 150 300>, windFrequency=<1.5 2 6>, windOffset=<0 -200 -200>}
   onGround_Moving {gravityOverride=<0 -50 -15>, windAmplitude=<50 150 300>, windFrequency=<1.5 2 6>, windOffset=<0 -200 -200>}
   onGround_Standing {gravityOverride=<0 -50 -15>, windAmplitude=<50 150 300>, windFrequency=<1.5 2 6>, windOffset=<0 -200 -200>}
   skydive_Parachuting {gravityOverride=<0 -300 -300>}
   swimming {windAmplitude=<50 -25 100>, windFrequency=<0.15 1 0.3>, windOffset=<0 -50 -50>}
  }
 }
 GravityOverride {type=Vector}
 InComponentSpace {type=boolean}
 InComponentSpaceLastFrame {type=boolean}
 InterpSpeed {type=number}
 LockedOnSimSpace {type=boolean}
 sim_space_rotation {type=Rotator}
]
