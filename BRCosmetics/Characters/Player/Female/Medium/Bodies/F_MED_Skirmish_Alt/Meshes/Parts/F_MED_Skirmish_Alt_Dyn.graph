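# Physics asset for the dynamic hair of the female medium Skirmish Alt player cosmetic body.
# Capsule bodies cover six bang bones, left and right bun bones, and the head.
# Each hair bone is constrained directly to the head in a star layout, with no chaining between strands.
# Twist is tightly limited throughout, while the buns are allowed wider swing than the bangs.
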PhysicsAsset /BRCosmetics/Characters/Player/Female/Medium/Bodies/F_MED_Skirmish_Alt/Meshes/Parts/F_MED_Skirmish_Alt_Dyn
bodies {
 dyn_hair_bangs_01 [{shapeType=Capsule, length=2.875, radius=1.6875, center=<2.2182 0 0.9725>, rotation=<rot 90 0 0>}]
 dyn_hair_bangs_02 [{shapeType=Capsule, length=2.875, radius=1.6875, center=<2.218 0 0.4849>, rotation=<rot 90 0 0>}]
 dyn_hair_bangs_03 [{shapeType=Capsule, length=2.6562, radius=1.9062, center=<2.218 0 0.6904>, rotation=<rot 90 0 0>}]
 dyn_hair_bangs_04 [{shapeType=Capsule, length=2.875, radius=1.6875, center=<2.218 0 0.5723>, rotation=<rot 90 0 0>}]
 dyn_hair_bangs_05 [{shapeType=Capsule, length=2.875, radius=1.6875, center=<2.218 0 0.5889>, rotation=<rot 90 0 0>}]
 dyn_hair_bangs_06 [{shapeType=Capsule, length=2.875, radius=1.6875, center=<2.218 0 0.6904>, rotation=<rot 90 0 0>}]
 dyn_hair_bun_l [{shapeType=Capsule, length=0.75, radius=1.8438, center=<1.1507 0 0>, rotation=<rot 90 0 0>}]
 dyn_hair_bun_r [{shapeType=Capsule, length=0.75, radius=1.8438, center=<1.1507 0 0>, rotation=<rot 90 0 0>}]
 head [{shapeType=Capsule, length=9.4792, radius=5.7642, center=<2.898 2.4583 0.0001>, rotation=<rot 90 0 0>}]
}
constraints [
 {bone1Name=dyn_hair_bun_l, bone2Name=head, swing1LimitDegrees=45, swing2LimitDegrees=45, twistLimitDegrees=5}
 {bone1Name=dyn_hair_bangs_01, bone2Name=head, swing1LimitDegrees=15, swing2LimitDegrees=30, twistLimitDegrees=5}
 {bone1Name=dyn_hair_bangs_02, bone2Name=head, swing1LimitDegrees=15, swing2LimitDegrees=30, twistLimitDegrees=5}
 {bone1Name=dyn_hair_bangs_03, bone2Name=head, swing1LimitDegrees=15, swing2LimitDegrees=30, twistLimitDegrees=5}
 {bone1Name=dyn_hair_bangs_04, bone2Name=head, swing1LimitDegrees=15, swing2LimitDegrees=30, twistLimitDegrees=5}
 {bone1Name=dyn_hair_bun_r, bone2Name=head, swing1LimitDegrees=45, swing2LimitDegrees=45, twistLimitDegrees=5}
 {bone1Name=dyn_hair_bangs_05, bone2Name=head, swing1LimitDegrees=15, swing2LimitDegrees=30, twistLimitDegrees=5}
 {bone1Name=dyn_hair_bangs_06, bone2Name=head, swing1LimitDegrees=15, swing2LimitDegrees=30, twistLimitDegrees=5}
]
