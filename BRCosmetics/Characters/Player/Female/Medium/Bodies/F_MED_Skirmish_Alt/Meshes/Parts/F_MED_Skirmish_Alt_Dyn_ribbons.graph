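# Physics asset for the dynamic hair ribbons on the female medium Skirmish Alt body cosmetic.
# Four strap chains (front and back, left and right) each run six capsule segments from a small master body constrained to the head.
# Chain joints use angular limits, wider at the root than along the strap, while master-to-head joints rotate freely.
# Head, upper spine and clavicle shapes provide body volumes around the ribbons.
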PhysicsAsset /BRCosmetics/Characters/Player/Female/Medium/Bodies/F_MED_Skirmish_Alt/Meshes/Parts/F_MED_Skirmish_Alt_Dyn_ribbons
bodies {
 clavicle_l [{shapeType=Capsule, length=11.1667, radius=7.4517, center=<5.3031 -1.3683 -1.9971>, rotation=<rot 82.0236 0 0>}]
 clavicle_r [{shapeType=Capsule, length=11.1667, radius=7.4517, center=<-5.3031 1.3683 1.9971>, rotation=<rot 82.0236 0 0>}]
 dyn_hair_strap_bk_l_01 [{shapeType=Capsule, length=2.4375, radius=1.0625, center=<3.4458 -0.0002 0>, rotation=<rot 90 0 0>}]
 dyn_hair_strap_bk_l_01_MSTR [{shapeType=Capsule, length=0.505, radius=0.505}]
 dyn_hair_strap_bk_l_02 [{shapeType=Capsule, length=3.5, radius=1.0625, center=<2.2181 0 0>, rotation=<rot 90 0 0>}]
 dyn_hair_strap_bk_l_03 [{shapeType=Capsule, length=3.5, radius=1.0625, center=<2.2181 0 0>, rotation=<rot 90 0 0>}]
 dyn_hair_strap_bk_l_04 [{shapeType=Capsule, length=3.5, radius=1.0625, center=<2.2181 0 0>, rotation=<rot 90 0 0>}]
 dyn_hair_strap_bk_l_05 [{shapeType=Capsule, length=3.5, radius=1.0625, center=<2.2181 0 0>, rotation=<rot 90 0 0>}]
 dyn_hair_strap_bk_l_06 [{shapeType=Capsule, length=3.5, radius=1.0625, center=<2.2181 0 0>, rotation=<rot 90 0 0>}]
 dyn_hair_strap_bk_r_01 [{shapeType=Capsule, length=2.4375, radius=1.0625, center=<3.4458 -0.0002 0>, rotation=<rot 90 0 0>}]
 dyn_hair_strap_bk_r_01_MSTR [{shapeType=Capsule, length=0.505, radius=0.505}]
 dyn_hair_strap_bk_r_02 [{shapeType=Capsule, length=3.5, radius=1.0625, center=<2.2181 0 0>, rotation=<rot 90 0 0>}]
 dyn_hair_strap_bk_r_03 [{shapeType=Capsule, length=3.5, radius=1.0625, center=<2.2181 0 0>, rotation=<rot 90 0 0>}]
 dyn_hair_strap_bk_r_04 [{shapeType=Capsule, length=3.5, radius=1.0625, center=<2.2181 0 0>, rotation=<rot 90 0 0>}]
 dyn_hair_strap_bk_r_05 [{shapeType=Capsule, length=3.5, radius=1.0625, center=<2.2181 0 0>, rotation=<rot 90 0 0>}]
 dyn_hair_strap_bk_r_06 [{shapeType=Capsule, length=3.5, radius=1.0625, center=<2.2181 0 0>, rotation=<rot 90 0 0>}]
 dyn_hair_strap_fr_l_01 [{shapeType=Capsule, length=2.4375, radius=1.0625, center=<3.4458 -0.0002 0>, rotation=<rot 90 0 0>}]
 dyn_hair_strap_fr_l_01_MSTR [{shapeType=Capsule, length=0.505, radius=0.505}]
 dyn_hair_strap_fr_l_02 [{shapeType=Capsule, length=3.5, radius=1.0625, center=<2.2181 0 0>, rotation=<rot 90 0 0>}]
 dyn_hair_strap_fr_l_03 [{shapeType=Capsule, length=3.5, radius=1.0625, center=<2.2181 0 0>, rotation=<rot 90 0 0>}]
 dyn_hair_strap_fr_l_04 [{shapeType=Capsule, length=3.5, radius=1.0625, center=<2.2181 0 0>, rotation=<rot 90 0 0>}]
 dyn_hair_strap_fr_l_05 [{shapeType=Capsule, length=3.5, radius=1.0625, center=<2.2181 0 0>, rotation=<rot 90 0 0>}]
 dyn_hair_strap_fr_l_06 [{shapeType=Capsule, length=3.5, radius=1.0625, center=<2.2181 0 0>, rotation=<rot 90 0 0>}]
 dyn_hair_strap_fr_r_01 [{shapeType=Capsule, length=2.4375, radius=1.0625, center=<3.4458 -0.0002 0>, rotation=<rot 90 0 0>}]
 dyn_hair_strap_fr_r_01_MSTR [{shapeType=Capsule, length=0.505, radius=0.505}]
 dyn_hair_strap_fr_r_02 [{shapeType=Capsule, length=3.5, radius=1.0625, center=<2.2181 0 0>, rotation=<rot 90 0 0>}]
 dyn_hair_strap_fr_r_03 [{shapeType=Capsule, length=3.5, radius=1.0625, center=<2.2181 0 0>, rotation=<rot 90 0 0>}]
 dyn_hair_strap_fr_r_04 [{shapeType=Capsule, length=3.5, radius=1.0625, center=<2.2181 0 0>, rotation=<rot 90 0 0>}]
 dyn_hair_strap_fr_r_05 [{shapeType=Capsule, length=3.5, radius=1.0625, center=<2.2181 0 0>, rotation=<rot 90 0 0>}]
 dyn_hair_strap_fr_r_06 [{shapeType=Capsule, length=3.5, radius=1.0625, center=<2.2181 0 0>, rotation=<rot 90 0 0>}]
 head [
  {shapeType=Sphere, radius=8.5625, center=<6.3729 -2.0759 0.0001>}
  {shapeType=Capsule, length=8.4792, radius=6.7642, center=<1.0557 -1.0594 0>, rotation=<rot 90 0 0>}
 ]
 spine_05 [
  {shapeType=Capsule, length=10.1979, radius=11.1079, center=<1.9336 0 0>}
  {shapeType=Capsule, length=4.5938, radius=7.625, center=<9.483 1.8562 0>}
 ]
}
constraints [
 {
  bone1Name=clavicle_l
  bone2Name=spine_05
  swing1LimitDegrees=45
  swing1Motion=Free
  swing2LimitDegrees=45
  swing2Motion=Free
  twistLimitDegrees=45
  twistMotion=Free
 }
 {
  bone1Name=clavicle_r
  bone2Name=spine_05
  swing1LimitDegrees=45
  swing1Motion=Free
  swing2LimitDegrees=45
  swing2Motion=Free
  twistLimitDegrees=45
  twistMotion=Free
 }
 {
  bone1Name=head
  bone2Name=spine_05
  swing1LimitDegrees=45
  swing1Motion=Free
  swing2LimitDegrees=45
  swing2Motion=Free
  twistLimitDegrees=45
  twistMotion=Free
 }
 {
  bone1Name=dyn_hair_strap_bk_l_01_MSTR
  bone2Name=head
  swing1LimitDegrees=45
  swing1Motion=Free
  swing2LimitDegrees=45
  swing2Motion=Free
  twistLimitDegrees=45
  twistMotion=Free
 }
 {bone1Name=dyn_hair_strap_bk_l_01, bone2Name=dyn_hair_strap_bk_l_01_MSTR, swing1LimitDegrees=45, swing2LimitDegrees=45, twistLimitDegrees=25}
 {bone1Name=dyn_hair_strap_bk_l_02, bone2Name=dyn_hair_strap_bk_l_01, swing1LimitDegrees=30, swing2LimitDegrees=30, twistLimitDegrees=15}
 {bone1Name=dyn_hair_strap_bk_l_03, bone2Name=dyn_hair_strap_bk_l_02, swing1LimitDegrees=30, swing2LimitDegrees=30, twistLimitDegrees=15}
 {bone1Name=dyn_hair_strap_bk_l_04, bone2Name=dyn_hair_strap_bk_l_03, swing1LimitDegrees=30, swing2LimitDegrees=30, twistLimitDegrees=15}
 {bone1Name=dyn_hair_strap_bk_l_05, bone2Name=dyn_hair_strap_bk_l_04, swing1LimitDegrees=30, swing2LimitDegrees=30, twistLimitDegrees=15}
 {bone1Name=dyn_hair_strap_bk_l_06, bone2Name=dyn_hair_strap_bk_l_05, swing1LimitDegrees=30, swing2LimitDegrees=30, twistLimitDegrees=15}
 {
  bone1Name=dyn_hair_strap_fr_l_01_MSTR
  bone2Name=head
  swing1LimitDegrees=45
  swing1Motion=Free
  swing2LimitDegrees=45
  swing2Motion=Free
  twistLimitDegrees=45
  twistMotion=Free
 }
 {bone1Name=dyn_hair_strap_fr_l_01, bone2Name=dyn_hair_strap_fr_l_01_MSTR, swing1LimitDegrees=45, swing2LimitDegrees=45, twistLimitDegrees=25}
 {bone1Name=dyn_hair_strap_fr_l_02, bone2Name=dyn_hair_strap_fr_l_01, swing1LimitDegrees=30, swing2LimitDegrees=30, twistLimitDegrees=15}
 {bone1Name=dyn_hair_strap_fr_l_03, bone2Name=dyn_hair_strap_fr_l_02, swing1LimitDegrees=30, swing2LimitDegrees=30, twistLimitDegrees=15}
 {bone1Name=dyn_hair_strap_fr_l_04, bone2Name=dyn_hair_strap_fr_l_03, swing1LimitDegrees=30, swing2LimitDegrees=30, twistLimitDegrees=15}
 {bone1Name=dyn_hair_strap_fr_l_05, bone2Name=dyn_hair_strap_fr_l_04, swing1LimitDegrees=30, swing2LimitDegrees=30, twistLimitDegrees=15}
 {bone1Name=dyn_hair_strap_fr_l_06, bone2Name=dyn_hair_strap_fr_l_05, swing1LimitDegrees=30, swing2LimitDegrees=30, twistLimitDegrees=15}
 {
  bone1Name=dyn_hair_strap_bk_r_01_MSTR
  bone2Name=head
  swing1LimitDegrees=45
  swing1Motion=Free
  swing2LimitDegrees=45
  swing2Motion=Free
  twistLimitDegrees=45
  twistMotion=Free
 }
 {bone1Name=dyn_hair_strap_bk_r_01, bone2Name=dyn_hair_strap_bk_r_01_MSTR, swing1LimitDegrees=45, swing2LimitDegrees=45, twistLimitDegrees=25}
 {bone1Name=dyn_hair_strap_bk_r_02, bone2Name=dyn_hair_strap_bk_r_01, swing1LimitDegrees=30, swing2LimitDegrees=30, twistLimitDegrees=15}
 {bone1Name=dyn_hair_strap_bk_r_03, bone2Name=dyn_hair_strap_bk_r_02, swing1LimitDegrees=30, swing2LimitDegrees=30, twistLimitDegrees=15}
 {bone1Name=dyn_hair_strap_bk_r_04, bone2Name=dyn_hair_strap_bk_r_03, swing1LimitDegrees=30, swing2LimitDegrees=30, twistLimitDegrees=15}
 {bone1Name=dyn_hair_strap_bk_r_05, bone2Name=dyn_hair_strap_bk_r_04, swing1LimitDegrees=30, swing2LimitDegrees=30, twistLimitDegrees=15}
 {bone1Name=dyn_hair_strap_bk_r_06, bone2Name=dyn_hair_strap_bk_r_05, swing1LimitDegrees=30, swing2LimitDegrees=30, twistLimitDegrees=15}
 {
  bone1Name=dyn_hair_strap_fr_r_01_MSTR
  bone2Name=head
  swing1LimitDegrees=45
  swing1Motion=Free
  swing2LimitDegrees=45
  swing2Motion=Free
  twistLimitDegrees=45
  twistMotion=Free
 }
 {bone1Name=dyn_hair_strap_fr_r_01, bone2Name=dyn_hair_strap_fr_r_01_MSTR, swing1LimitDegrees=45, swing2LimitDegrees=45, twistLimitDegrees=25}
 {bone1Name=dyn_hair_strap_fr_r_02, bone2Name=dyn_hair_strap_fr_r_01, swing1LimitDegrees=30, swing2LimitDegrees=30, twistLimitDegrees=15}
 {bone1Name=dyn_hair_strap_fr_r_03, bone2Name=dyn_hair_strap_fr_r_02, swing1LimitDegrees=30, swing2LimitDegrees=30, twistLimitDegrees=15}
 {bone1Name=dyn_hair_strap_fr_r_04, bone2Name=dyn_hair_strap_fr_r_03, swing1LimitDegrees=30, swing2LimitDegrees=30, twistLimitDegrees=15}
 {bone1Name=dyn_hair_strap_fr_r_05, bone2Name=dyn_hair_strap_fr_r_04, swing1LimitDegrees=30, swing2LimitDegrees=30, twistLimitDegrees=15}
 {bone1Name=dyn_hair_strap_fr_r_06, bone2Name=dyn_hair_strap_fr_r_05, swing1LimitDegrees=30, swing2LimitDegrees=30, twistLimitDegrees=15}
]
