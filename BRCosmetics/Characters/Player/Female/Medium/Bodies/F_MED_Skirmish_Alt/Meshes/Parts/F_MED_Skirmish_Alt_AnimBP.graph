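# Post-process-style cosmetic AnimBP for the F_MED_Skirmish_Alt body part, adding dynamic hair and ribbon motion to the pose copied from its parent mesh.
# Four hair-strap bones get Perlin-noise additive rotations, gated by bRigidBodyEnabled, before two chained neck-relative RigidBody solvers.
# Each update, a deprecated state function turns per-state gravity and wind parameters into both solvers' gravity, scaled for front hair.
# Debug draw nodes are unconnected.
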AnimBlueprint /BRCosmetics/Characters/Player/Female/Medium/Bodies/F_MED_Skirmish_Alt/Meshes/Parts/F_MED_Skirmish_Alt_AnimBP
target_skeleton=&/F_MED_Skirmish_Alt_Skeleton
graphs [
 AnimGraph {nodes [
   n1 = Get bRigidBodyEnabled()
   n2 = "Transform (Modify) Bone - Bone: dyn_hair_strap_fr_l_01_MSTR"!(
    ComponentPose=Local To Component(LocalPose=Copy Pose From Mesh() {bUseAttachedParent=true})
    bAlphaBoolEnabled:bool=@n1
    Translation="0,0,0"
    Rotation=Get HairRotator_01() -> (HairRotator_01:Rotator)
    Scale="1,1,1"
   ) -> (Pose) {
    alphaInputType=Bool
    rotationMode=BMM_Additive
    rotationSpace=BCS_BoneSpace
    alphaBoolBlend {blendOption=Linear}
    boneToModify {boneName=dyn_hair_strap_fr_l_01_MSTR}
   }
   n3 = "Transform (Modify) Bone - Bone: dyn_hair_strap_bk_l_01_MSTR"!(ComponentPose=@n2, bAlphaBoolEnabled:bool=@n1, Translation="0,0,0", Rotation=Get HairRotator_02() -> (HairRotator_02:Rotator), Scale="1,1,1") -> (Pose) {
    alphaInputType=Bool
    rotationMode=BMM_Additive
    rotationSpace=BCS_BoneSpace
    alphaBoolBlend {blendOption=Linear}
    boneToModify {boneName=dyn_hair_strap_bk_l_01_MSTR}
   }
   n4 = "Transform (Modify) Bone - Bone: dyn_hair_strap_fr_r_01_MSTR"!(ComponentPose=@n3, bAlphaBoolEnabled:bool=@n1, Translation="0,0,0", Rotation=Get HairRotator_03() -> (HairRotator_03:Rotator), Scale="1,1,1") -> (Pose) {
    alphaInputType=Bool
    rotationMode=BMM_Additive
    rotationSpace=BCS_BoneSpace
    alphaBoolBlend {blendOption=Linear}
    boneToModify {boneName=dyn_hair_strap_fr_r_01_MSTR}
   }
   n5 = "Transform (Modify) Bone - Bone: dyn_hair_strap_bk_r_01_MSTR"!(ComponentPose=@n4, bAlphaBoolEnabled:bool=@n1, Translation="0,0,0", Rotation=Get HairRotator_04() -> (HairRotator_04:Rotator), Scale="1,1,1") -> (Pose) {
    alphaInputType=Bool
    rotationMode=BMM_Additive
    rotationSpace=BCS_BoneSpace
    alphaBoolBlend {blendOption=Linear}
    boneToModify {boneName=dyn_hair_strap_bk_r_01_MSTR}
   }
   n6 = RigidBody!(ComponentPose=@n5, Alpha, OverrideWorldGravity:Vector=Get GravityOverride_FrontHair() -> (GravityOverride_FrontHair:Vector), ExternalForce) -> (Pose) {
    bClampLinearTranslationLimitToRefPose=true
    bOverrideWorldGravity=true
    overridePhysicsAsset=&/F_MED_Skirmish_Alt_Dyn
    simulationSpace=BaseBoneSpace
    alphaBoolBlend {blendOption=Linear}
    baseBoneRef {boneName=neck_01}
    componentAppliedLinearAccClamp=<3000 3000 3000>
    componentLinearAccScale=<0.4 0.3>
    componentLinearVelScale=<0.4 0.3>
    overrideWorldGravity {z=-980}
    simSpaceSettings {
     maxAngularAcceleration=1000
     maxAngularVelocity=0.3
     maxLinearAcceleration=700
     maxLinearVelocity=500
     velocityScaleZ=0
     worldAlpha=0.1
     externalLinearDragV=<0.5 0.5>
    }
   }
   n7 = RigidBody(ComponentPose=@n6, OverrideWorldGravity=Get GravityOverride()) {
    bClampLinearTranslationLimitToRefPose=true
    bOverrideWorldGravity=true
    overridePhysicsAsset=&/F_MED_Skirmish_Alt_Dyn_ribbons
    simulationSpace=BaseBoneSpace
    alphaBoolBlend {blendOption=Linear}
    baseBoneRef {boneName=neck_01}
    componentAppliedLinearAccClamp=<3000 3000 3000>
    componentLinearAccScale {x=0.1}
    componentLinearVelScale {x=0.1}
    overrideWorldGravity {z=-980}
    simSpaceSettings {
     maxAngularAcceleration=1000
     maxAngularVelocity=3
     maxLinearAcceleration=700
     maxLinearVelocity=500
     velocityScaleZ=0
     worldAlpha=1
     externalLinearDragV=<0.2 0.2>
    }
   }
   Output Pose(Result=Component To Local(ComponentPose=@n7)) {layerGroup=None}
  ]}
 EventGraph {nodes [
   Event BlueprintUpdateAnimation()
   > n1 = Sequence!(execute) -> (then_0, then_1, then_2:Exec)
   > n2 = Gravity Override State Function DEPRECATED(
    CharacterPartAnimInstance=Self-Reference()
    GameTimeInSeconds=GetGameTimeInSeconds()
    GravityOverrideParams=Get Gravity Override Params()
    DebugJointName=head
   )
   > n3 = Set GravityOverride(GravityOverride=@n2)
   > Set GravityOverride_FrontHair(GravityOverride_FrontHair:Vector=vector * vector(A:Vector=@n3, B:Vector="0.4,0.4,2") -> (ReturnValue:Vector) {is=title K2Node_CommutativeAssociativeBinaryOperator}) -> (Output_Get:Vector)
   n4 = GetGameTimeInSeconds()
   n5 = RandomRotatorFunction!(
    execute=@n1.then_1
    GameTimeInSeconds:double=@n4
    PerlinScale:double=1
    Frequency:Rotator="5,3,0"
    Amplitude:Rotator="50,50,0"
    Offset:Rotator="0,-45, 0"
    TimeOffset:double=0
   ) -> (then, RandomRotator:Rotator)
   > Set HairRotator_01(HairRotator_01:Rotator=@n5) -> (Output_Get:Rotator)
   > n6 = RandomRotatorFunction(GameTimeInSeconds=@n4, PerlinScale=1, Frequency="2.2,3.7, 0", Amplitude="50,50, 0", Offset="0,-45, 0", TimeOffset=0)
   > Set HairRotator_02(HairRotator_02:Rotator=@n6) -> (Output_Get:Rotator)
   > n7 = RandomRotatorFunction(GameTimeInSeconds=@n4, PerlinScale=1, Frequency="3.6,1.9, 0", Amplitude="50,50,0", Offset="0,-45, 0", TimeOffset=0)
   > Set HairRotator_03(HairRotator_03:Rotator=@n7) -> (Output_Get:Rotator)
   > n8 = RandomRotatorFunction(GameTimeInSeconds=@n4, PerlinScale=1, Frequency="4.3,2.8, 0", Amplitude="50,50, 0", Offset="0,-45, 0", TimeOffset=0)
   > Set HairRotator_04(HairRotator_04:Rotator=@n8) -> (Output_Get:Rotator)
   DrawVectorFromJointFunction!(
    execute
    SkelMeshComponent:Scene Component Object Reference
    JointName:Name=head
    LineEnd:Vector=@n3
    JointOffset:Vector="0, 0, 0"
    LineColor:Linear Color Structure=(R=1,G=0,B=0,A=1)
    Thickness:double=2
    Foreground:EDrawDebugSceneDepthPriorityGroup Enum=Foreground
   ) -> (then)
   DrawJointDebugFunction!(
    execute
    SkelMeshComponent:Scene Component Object Reference
    JointOffset:Vector="0, 0, 0"
    JointName:Array of Names=Make Array("[0]":Name=root) -> (Array:Array of Names)
    Scale:double=10
    Duration:double=0
    Thickness:double=0.7
    DebugDraw:bool=true
    Foreground:EDrawDebugSceneDepthPriorityGroup Enum=Foreground
   ) -> (then)
  ]}
]
variables [
 Gravity Override Params {
  type=GravityOverrideParamsStruct
  default {
   emoteOrMelee {gravityOverride=<0 -50 -15>, windAmplitude=<50 25 50>, windFrequency=<0.1 0.2 0.15>, windOffset=<0 -50 -50>}
   front_End {gravityOverride=<0 -50 -15>, windAmplitude=<50 25 50>, windFrequency=<0.1 0.2 0.15>, windOffset=<0 -50 -50>}
   onGround_Moving {gravityOverride=<0 -50 -15>, windAmplitude=<50 25 50>, windFrequency=<0.1 0.2 0.15>, windOffset=<0 -50 -50>}
   onGround_Standing {gravityOverride=<0 -50 -15>, windAmplitude=<50 25 50>, windFrequency=<0.1 0.2 0.15>, windOffset=<0 -50 -50>}
   skydive_Parachuting {gravityOverride=<0 -300 -300>}
   swimming {windAmplitude=<50 -25 100>, windFrequency=<0.15 1 0.3>, windOffset=<0 -50 -50>}
  }
 }
 GravityOverride {type=Vector}
 GravityOverride_FrontHair {type=Vector}
 HairRotator_01 {type=Rotator}
 HairRotator_02 {type=Rotator}
 HairRotator_03 {type=Rotator}
 HairRotator_04 {type=Rotator}
]
